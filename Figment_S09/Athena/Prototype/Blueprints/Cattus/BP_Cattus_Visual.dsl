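; BP_Cattus_Visual is a visual actor for the Cattus prototype that registers its static mesh part with a Test_Cattus manager actor.
; On BeginPlay, only on the remote (non-authority) side, it finds all Test_Cattus actors and checks whether the retrieved one is valid.
; If that actor is invalid, a short delay runs before RegisterMeshwithManager, with no repeat search shown.
; Tick, overlap and construction logic are empty.
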(event EventBeginPlay
  (Networking|SwitchHasAuthority
    (:Remote
      (bind _outactors (Actor|GetAllActorsOfClass "/Figment_S09/Athena/Prototype/Blueprints/Cattus/Test_Cattus.Test_Cattus_C"))
      (Utilities|IsValid (Utilities|Array|Get(acopy) _outactors)
        (:Is%20Valid
          _)
        (:Is%20Not%20Valid
          (Utilities|FlowControl|Delay 2.0)))
      (Class|TestCattus|RegisterMeshwithManager (Utilities|Array|Get(acopy) _outactors) (Variables|Default|GetCattus_Static_Part)))))

(event Collision|EventActorBeginOverlap
  (params (OtherActor :type Actor%20Object%20Reference)))

(event EventTick
  (params (DeltaSeconds :type Float%20%28single-precision%29)))

(fn ConstructionScript)
