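# Translucent, two-sided lit water-flow effect for the Cattus prototype, usable on skeletal meshes.
# Panning caustics, distorted by Perlin noise and scene normals, combine with bubble splashes, curve-atlas edge ramps and depth fade into a mask for opacity and refraction.
# Vertex colors vary the pattern, and blue minus MPC WaterFlowOpacity drives mask erosion and vertex displacement alongside a camera offset.
# Glow uses animated dots and time-of-day color, with distance-based adjustment.
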Material /Figment_S09/Athena/Prototype/Blueprints/Cattus/FX/M_Material
graph {
 properties {
  bUsedWithMeshDeformer=true
  bUsedWithSkeletalMesh=true
  blendMode=BLEND_Translucent
  pixelDepthOffsetMode=PDOM_Legacy
  refractionMethod=RM_IndexOfRefraction
  translucencyPass=MTP_BeforeDOF
  twoSided=true
  usedShadingModels=Default Lit
 }
 nodes [
  n1 = ScalarParameter() {defaultValue=75, parameterName=Offset}
  n2 = TextureCoordinate() {uTiling=2, vTiling=2}
  n3 = Multiply(A=Constant3Vector() {constant=<rgba 0 -180 0 0>}, B=ComponentMask(None=Subtract(A=@n2)) {g=true})
  n4 = Saturate(None=Subtract(
    A=VertexColor().B
    B=CollectionParameter() {collection=&/MPC_CattusShader, parameterName=WaterFlowOpacity}
   ))
  Add(
   A=MaterialFunctionCall(Offset Amount=@n1) {materialFunction=/Engine/Functions/Engine_MaterialFunctions02/WorldPositionOffset/CameraOffset}.Clamped Camera Offset
   B=Multiply(A=Transform(None=@n3), B=@n4)
  )
  n5 = Constant2Vector() {g=6, r=2}
  n6 = MaterialFunctionCall(Float2=TextureCoordinate()) {materialFunction=&BreakOutFloat2Components}
  n7 = ScalarParameter() {defaultValue=2.2, parameterName=Exp}
  n8 = Multiply(A=AppendVector(A=Power(Base=@n6.R, Exp=@n7), B=@n6.G), B=LinearInterpolate(Alpha=VertexColor().G) {constA=0.3, constB=0.5}) {constB=0.4}
  n9 = Multiply(A=TextureSample(UVs=Multiply(A=@n8, B=@n5)) {texture=/Game/Effects/Fort_Effects/Textures/Water/T_MaskeSplashes_Bubbles1_A}.R) {constB=0.25}
  n10 = VertexColor()
  n11 = Add(A=@n8, B=AppendVector(A=@n10.G, B=@n10.G))
  n12 = VertexColor()
  n13 = Multiply(A=LinearInterpolate(Alpha=@n12.R) {constA=0.1, constB=0.2}, B=Constant2Vector() {r=1})
  n14 = Add(A=TextureCoordinate() {coordinateIndex=1, uTiling=24, vTiling=2}, B=@n12.R)
  n15 = Subtract(A=ComponentMask(None=TextureSample(UVs=@n14) {texture=/Game/Global/Textures/Default/Noise/3xPerlin_M}.RGB) {g=true, r=true}) {constB=0.4}
  n16 = SceneTexture() {sceneTextureId=PPI_WorldNormal}
  n17 = ComponentMask(None=Transform(None=ComponentMask(None=@n16.Color) {b=true, g=true, r=true}) {transformSourceType=TRANSFORMSOURCE_World, transformType=TRANSFORM_View}) {g=true, r=true}
  n18 = Constant()
  n19 = ShadingPathSwitch(Default=Add(A=Multiply(A=@n15) {constB=0.1}, B=Multiply(A=@n17) {constB=0.025}), Forward=@n18)
  n20 = TextureSample(UVs=Add(A=Panner(Coordinate=@n11, Speed=@n13) {speedX=0.1}, B=FeatureLevelSwitch(Default=@n19, ES2=@n18, ES3_1=@n18, SM4_REMOVED=@n19))) {texture=/Figment_S09/Effects/Fort_Effects/Textures/Water/T_Caustics_Softer}
  n21 = MaterialFunctionCall(Float2=TextureCoordinate()) {materialFunction=&BreakOutFloat2Components}
  n22 = CurveAtlasRowParameter(CurveTime=@n21.R) {atlas=/Game/Athena/LinearColorCurves/TempAthenaCurveAtlas, curve=/Game/Athena/Prototype/Blueprints/Cattus/FX/Curve_Ramps, defaultValue=4}
  n23 = CurveAtlasRowParameter(CurveTime=@n21.G) {atlas=&TempAthenaCurveAtlas, curve=&Curve_Ramps, defaultValue=4}
  n24 = Multiply(A=Multiply(A=VertexColor().G, B=Add(A=@n9, B=@n20.R)), B=Multiply(A=@n22.R, B=@n23.G))
  n25 = Multiply(A=@n24, B=DepthFade(FadeDistance=@n1) {fadeDistanceDefault=256})
  n26 = ComponentMask(None=TextureCoordinate()) {r=true}
  n27 = Multiply(A=FeatureLevelSwitch(Default=@n25, ES2=@n24, ES3_1=@n24, SM4_REMOVED=@n25), B=Subtract(A=Add(A=@n26) {constB=0.1}, B=@n4))
  n28 = Multiply(A=@n27) {constB=16}
  n29 = Distance(A=CameraPositionWS(), B=WorldPosition().XYZ)
  n30 = Saturate(None=Subtract(A=Divide(A=@n29) {constA=200000, constB=150000}) {constB=0.25})
  LinearInterpolate(A=@n27, B=@n28, Alpha=@n30)
  LinearInterpolate(Alpha=@n27) {constA=1, constB=0.96}
  n31 = TextureCoordinate()
  n32 = VertexColor()
  n33 = Add(A=Add(A=Add(A=@n31, B=@n32.G), B=Time()), B=@n20.R)
  n34 = CurveAtlasRowParameter(CurveTime=Multiply(A=TextureSample(UVs=@n33) {texture=/Figment_S09/Missions/Secondary/Outpost/IronCity/Pods/Textures/T_Ironcity_Pods_Dots}.B, B=@n20.R)) {atlas=&TempAthenaCurveAtlas, curve=&Curve_Ramps, defaultValue=4}
  n35 = MaterialFunctionCall() {materialFunction=&Color_Time_of_Day}
  n36 = VectorParameter() {parameterName=Color, defaultValue=<rgba 0.542 1 0.8687 0>}
  Multiply(A=Multiply(A=Multiply(A=@n34.B, B=@n35.Result)) {constB=400}, B=LinearInterpolate(A=@n36.RGB, B=Multiply(A=@n36.RGB, B=@n28), Alpha=@n30))
  ScalarParameter() {defaultValue=1, parameterName=Test}
 ]
}
