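; Prototype Cattus event actor: a creature with periodic breathing and a player-tracking eye, staged by the CatEye stateful event.
; Stage 1 starts timed breaths with camera shake, sound and collision boxes that launch pawns and vehicles.
; Stage 2 reveals a hidden mesh and enables server target selection from overlaps, damage and random retargeting.
; Clients follow the replicated target with randomized drift, interpolating eye material parameters and motion-driven audio on tick.
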(event Custom|RegisterMeshWithManager
  (params (Mesh :type Static%20Mesh%20Component%20Object%20Reference))
  (bind _returnvalue (Rendering|Material|CreateDynamicMaterialInstance Mesh 1 0 "None" :!class /Script/Engine.PrimitiveComponent))
  (bind _output_get (Variables|Default|SetMID_MovingMaterial _returnvalue))
  (Rendering|Material|SetVectorParameterValue _output_get "Target" (Math|Conversions|ToLinearColor(Vector) (+ (Transformation|GetActorLocation) (Variables|Default|GetTarget))) :!class /Script/Engine.MaterialInstanceDynamic))

(event Custom|GetAndActivateFoundation)

(event EventBeginPlay
  (Collision|SetCollisionEnabled (Variables|Default|GetBox) "NoCollision")
  (Networking|SwitchHasAuthority
    (:Authority
      (bind (_asynctaskproxy _payload) (GameplayEventRouter|StartListeningToStatefulEvent (GameplayEventRouter|GetGlobalEventRouter) "/Figment_S09/Gameplay/EventRouter/ST_Figment_CatEye_Event.ST_Figment_CatEye_Event")
        (:OnEventReceived
          _)
        (:OnSavedEventStateExists
          _))
      (switch Utilities|FlowControl|Switch|SwitchonEnum_CatEye_Event_State (Utilities|Struct|BreakSTFigmentCatEyeEvent _payload)
        (:NewEnumerator1
          (Variables|Default|SetStage1Active true))
        (:NewEnumerator2
          (if (not (Variables|Default|GetStage1Active))
            (Variables|Default|SetStage1Active true))
          (Variables|Default|SetStage2Active true)
          (CallFunction|GetandActivateFoundation))))))

(event Custom|TestActivateCatBreathing
  (Variables|Default|SetStage1Active true))

(event Custom|TestActivateCat
  (Variables|Default|SetStage2Active true)
  (Variables|Default|SetStage1Active true))

(event Custom|ActivateCat
  (Variables|Default|SetStage2Active true)
  (Variables|Default|SetStage1Active true))

(event OnComponentBeginOverlap(Sphere)
  (params (OverlappedComponent :type Primitive%20Component%20Object%20Reference) (OtherActor :type Actor%20Object%20Reference) (OtherComp :type Primitive%20Component%20Object%20Reference) (OtherBodyIndex :type Integer) (bFromSweep :type Boolean) (SweepResult :type Hit%20Result%20Structure%20%28by%20ref%29))
  (Networking|SwitchHasAuthority
    (:Authority
      (bind _asfort_pawn (Utilities|Casting|CastToFortPawn OtherActor))
      (Variables|Default|SetTrackedTarget _asfort_pawn)
      (if (not (Utilities|Time|IsTimerActivebyHandle (Variables|Default|GetChangeTargetsTimer)))
        (bind _returnvalue (Utilities|Time|SetTimerbyEvent (EventDispatchers|CreateEvent :!event Custom|ChangeTargetsInArray) (Math|Random|RandomFloatinRange 3.0 6.0)))
        (Variables|Default|SetChangeTargetsTimer _returnvalue)))))

(event OnComponentEndOverlap(Sphere)
  (params (OverlappedComponent :type Primitive%20Component%20Object%20Reference) (OtherActor :type Actor%20Object%20Reference) (OtherComp :type Primitive%20Component%20Object%20Reference) (OtherBodyIndex :type Integer))
  (Networking|SwitchHasAuthority
    (:Authority
      (if (Utilities|Equal(Object) OtherActor (Variables|Default|GetTrackedTarget))
        (CallFunction|ChangeTargetsinArray)))))

(event Game|Damage|EventAnyDamage
  (params (Damage :type Float%20%28single-precision%29) (DamageType :type Damage%20Type%20Object%20Reference) (InstigatedBy :type Controller%20Object%20Reference) (DamageCauser :type Actor%20Object%20Reference))
  (Utilities|IsValid InstigatedBy
    (:Is%20Valid
      (bind _asfort_pawn (Utilities|Casting|CastToFortPawn (Pawn|GetControlledPawn InstigatedBy)))
      (Variables|Default|SetTrackedTarget _asfort_pawn))))

(event Custom|ChangeTargetsInArray
  (bind _sphere (Variables|Default|GetSphere))
  (bind _overlappingactors (Collision|GetOverlappingActors _sphere "/Script/FortniteGame.FortPawn" :!class /Script/Engine.PrimitiveComponent))
  (bind _returnvalue (Utilities|Array|Length _overlappingactors))
  (bind _returnvalue_2 (- _returnvalue 1))
  (bind _returnvalue_3 (Math|Random|RandomInteger _returnvalue_2))
  (bind _output (Utilities|Array|Get(acopy) _overlappingactors _returnvalue_3))
  (Utilities|Time|ClearandInvalidateTimerbyHandle (Variables|Default|GetChangeTargetsTimer))
  (if (> _returnvalue 0)
    (Utilities|IsValid _output
      (:Is%20Valid
        (Variables|Default|SetTrackedTarget _output)
        (if (not (Utilities|Time|IsTimerActivebyHandle (Variables|Default|GetChangeTargetsTimer)))
          (bind _returnvalue_4 (Utilities|Time|SetTimerbyEvent (EventDispatchers|CreateEvent :!event Custom|ChangeTargetsInArray) (Math|Random|RandomFloatinRange 3.0 6.0)))
          (Variables|Default|SetChangeTargetsTimer _returnvalue_4))))
    (else
      (Variables|Default|SetTrackedTarget 0))))

(event Custom|UpdatePupil
  (Utilities|IsValid (Variables|Default|GetTrackedTarget)
    (:Is%20Not%20Valid
      (Rendering|Material|SetScalarParameterValue (Variables|Default|GetMID_MovingMaterial) "PupilScale" 0.75 :!class /Script/Engine.MaterialInstanceDynamic))))

(event Custom|UpdateTarget
  (bind _trackedtarget (Variables|Default|GetTrackedTarget))
  (bind _returnvalue (Transformation|GetActorLocation _trackedtarget))
  (Networking|SwitchHasAuthority
    (:Remote
      (Utilities|IsValid _trackedtarget
        (:Is%20Valid
          (Actor|Tick|SetActorTickEnabled true)
          (if (< (Math|Vector|VectorLength (- _returnvalue (Variables|Default|GetTarget))) (Variables|Default|GetCloseRangeMin))
            (Utilities|Time|SetTimerbyEvent (EventDispatchers|CreateEvent :!event Custom|UpdateTarget) (Math|Random|RandomFloatinRange 0.1 0.6))
            (else
              (Utilities|Time|SetTimerbyEvent (EventDispatchers|CreateEvent :!event Custom|UpdateTarget) (Math|Random|RandomFloatinRange 0.6 2.0))))
          (bind _output_get (Variables|Default|SetTarget (+ _returnvalue (Math|Vector|MakeVector (Math|Random|RandomFloatinRange (Variables|Default|GetMinDriftWhenOnTarget) (Variables|Default|GetMaxDriftWhenOnTarget)) (Math|Random|RandomFloatinRange (Variables|Default|GetMinDriftWhenOnTarget) (Variables|Default|GetMaxDriftWhenOnTarget)) (Math|Random|RandomFloatinRange (Variables|Default|GetMinDriftWhenOnTarget) (Variables|Default|GetMaxDriftWhenOnTarget))))))
          (Development|PrintString (Utilities|String|ToString(Vector) _output_get) false false))
        (:Is%20Not%20Valid
          (bind _output_get_2 (Variables|Default|SetTarget (+ (Variables|Default|GetTarget) (Math|Vector|MakeVector (Math|Random|RandomFloatinRange (Variables|Default|GetMinDriftWhenOnTarget) (Variables|Default|GetMaxDriftWhenOnTarget)) (Math|Random|RandomFloatinRange (Variables|Default|GetMinDriftWhenOnTarget) (Variables|Default|GetMaxDriftWhenOnTarget)) (Math|Random|RandomFloatinRange (Variables|Default|GetMinDriftWhenOnTarget) (Variables|Default|GetMaxDriftWhenOnTarget))))))
          (Development|PrintString (Utilities|String|ToString(Vector) _output_get_2) false false)
          (Rendering|Material|SetScalarParameterValue (Variables|Default|GetMID_MovingMaterial) "TrackPlayer" 0.0 :!class /Script/Engine.MaterialInstanceDynamic)
          (Utilities|Time|SetTimerbyEvent (EventDispatchers|CreateEvent :!event Custom|UpdateTarget) (Math|Random|RandomFloatinRange 3.0 4.0))))
      (Utilities|FlowControl|DoOnce false
        (:Completed
          (CallFunction|UpdatePupil)
          (if (not (Audio|Components|Audio|IsPlaying (Variables|Default|GetUpdateTargetAudio)))
            (Audio|Components|Audio|Play (Variables|Default|GetUpdateTargetAudio) 0.0)
            (Audio|Parameter|SetIntegerParameter (Variables|Default|GetTargetAudioLoop) "Active" 1 :!class /Script/Engine.AudioComponent)))))))

(event EventTick
  (params (DeltaSeconds :type Float%20%28single-precision%29))
  (bind _output_get (Variables|Default|SetSmoothedTarget (Math|Interpolation|VInterpTo (Variables|Default|GetSmoothedTarget) (Variables|Default|GetTarget) DeltaSeconds (Variables|Default|GetInterpSpeed))))
  (Rendering|Material|SetVectorParameterValue (Variables|Default|GetMID_MovingMaterial) "Target" (Math|Conversions|ToLinearColor(Vector) _output_get) :!class /Script/Engine.MaterialInstanceDynamic)
  (CallFunction|UpdateTargetAudioLoop)
  (bind _output_get_2 (Variables|Default|SetIsTracking (Math|Interpolation|FInterpTo (Variables|Default|GetIsTracking) (select (Utilities|IsValid (Variables|Default|GetTrackedTarget) :!class /Script/Engine.KismetSystemLibrary) 1.0 0.0) DeltaSeconds 3.0)))
  (Rendering|Material|SetScalarParameterValue (Variables|Default|GetMID_MovingMaterial) "PupilScale" (Math|Float|Lerp (Variables|Default|GetBigPupilScale) 1.8 _output_get_2) :!class /Script/Engine.MaterialInstanceDynamic)
  (Rendering|Material|SetScalarParameterValue (Variables|Default|GetMID_MovingMaterial) "TrackPlayer" _output_get_2 :!class /Script/Engine.MaterialInstanceDynamic)
  (if (== (Variables|Default|GetIsTracking) 0.0)
    (Actor|Tick|SetActorTickEnabled false)
    (Audio|Parameter|SetIntegerParameter (Variables|Default|GetTargetAudioLoop) "Active" 0 :!class /Script/Engine.AudioComponent)))

(event Custom|TestCatBreathe
  (bind _box1 (Variables|Default|GetBox1))
  (bind _box4 (Variables|Default|GetBox4))
  (bind _box3 (Variables|Default|GetBox3))
  (CallFunction|BreathShake)
  (Timeline|Timeline_0 PlayFromStart
    (:NewTrack_1
      (Components|Activation|Activate (Variables|Default|GetP_Cattus_Breathing_02) false)
      (Utilities|FlowControl|Delay 0.5)
      (Utilities|FlowControl|Sequence
        (:then_0
          (Collision|SetCollisionEnabled _box1 "QueryAndPhysics"))
        (:then_1
          (Utilities|FlowControl|Delay 0.5)
          (Collision|SetCollisionEnabled _box1 "NoCollision"))))
    (:NewTrack_2
      (Components|Activation|Activate (Variables|Default|GetP_Cattus_Breathing_04) false)
      (Utilities|FlowControl|Delay 0.5)
      (Utilities|FlowControl|Sequence
        (:then_0
          (Collision|SetCollisionEnabled _box3 "QueryAndPhysics"))
        (:then_1
          (Utilities|FlowControl|Delay 0.5)
          (Collision|SetCollisionEnabled _box3 "NoCollision"))))
    (:NewTrack_3
      (Components|Activation|Activate (Variables|Default|GetP_Cattus_Breathing_05) false)
      (Utilities|FlowControl|Delay 0.5)
      (Utilities|FlowControl|Sequence
        (:then_0
          (Collision|SetCollisionEnabled _box4 "QueryAndPhysics"))
        (:then_1
          (Utilities|FlowControl|Delay 0.5)
          (Collision|SetCollisionEnabled _box4 "NoCollision"))))
    (:NewTrack_4
      (Components|Activation|Activate (Variables|Default|GetP_Cattus_Debris_Falling2) false))
    (:NewTrack_5
      (Components|Activation|Activate (Variables|Default|GetP_Cattus_Debris_Falling3) false))))

(event Custom|CE_Breathe
  (bind _box1 (Variables|Default|GetBox1))
  (bind _box4 (Variables|Default|GetBox4))
  (bind _box3 (Variables|Default|GetBox3))
  (Utilities|Time|SetTimerbyEvent (EventDispatchers|CreateEvent :!event Custom|CE_Breathe) 3e+01)
  (CallFunction|BreathShake)
  (Timeline|Timeline_0 PlayFromStart))

(event Custom|BreathShake
  (bind _returnvalue (Fortnite|GetLocalFortPlayerControllers true true))
  (bind _output (Utilities|Array|Get(acopy) _returnvalue))
  (bind _box (Variables|Default|GetBox))
  (Components|Activation|Activate (Variables|Default|GetP_Cattus_Breathing_01) false)
  (Utilities|FlowControl|Sequence
    (:then_0
      (Utilities|IsValid _output
        (:Is%20Valid
          (if (<= (Transformation|GetDistanceTo (Pawn|GetControlledPawn _output) self) (Variables|Default|GetBreathShakeRadius))
            (GameplayCue|ExecuteGameplayCue%28Burst%29Local (Pawn|GetControlledPawn _output) (Variables|Default|GetShakeBurst) (Ability|GameplayCue|MakeGameplayCueParameters 0.0))))))
    (:then_1
      (Audio|PlaySoundatLocation "/Figment_S09/Athena/Sounds/Events/Cattus/Cattus_Exist_Cue.Cattus_Exist_Cue" (Transformation|GetWorldLocation (Variables|Default|GetTargetAudioLoop)))
      (Utilities|FlowControl|Delay 0.5)
      (Utilities|FlowControl|Sequence
        (:then_0
          (Collision|SetCollisionEnabled _box "QueryAndPhysics"))
        (:then_1
          (Utilities|FlowControl|Delay 0.5)
          (Collision|SetCollisionEnabled _box "NoCollision"))))))

(event OnComponentBeginOverlap(Box)
  (params (OverlappedComponent :type Primitive%20Component%20Object%20Reference) (OtherActor :type Actor%20Object%20Reference) (OtherComp :type Primitive%20Component%20Object%20Reference) (OtherBodyIndex :type Integer) (bFromSweep :type Boolean) (SweepResult :type Hit%20Result%20Structure%20%28by%20ref%29))
  (CallFunction|LaunchTargets OtherActor))

(event OnComponentBeginOverlap(Box1)
  (params (OverlappedComponent :type Primitive%20Component%20Object%20Reference) (OtherActor :type Actor%20Object%20Reference) (OtherComp :type Primitive%20Component%20Object%20Reference) (OtherBodyIndex :type Integer) (bFromSweep :type Boolean) (SweepResult :type Hit%20Result%20Structure%20%28by%20ref%29))
  (CallFunction|LaunchTargets OtherActor))

(event OnComponentBeginOverlap(Box2)
  (params (OverlappedComponent :type Primitive%20Component%20Object%20Reference) (OtherActor :type Actor%20Object%20Reference) (OtherComp :type Primitive%20Component%20Object%20Reference) (OtherBodyIndex :type Integer) (bFromSweep :type Boolean) (SweepResult :type Hit%20Result%20Structure%20%28by%20ref%29))
  (CallFunction|LaunchTargets OtherActor))

(event OnComponentBeginOverlap(Box3)
  (params (OverlappedComponent :type Primitive%20Component%20Object%20Reference) (OtherActor :type Actor%20Object%20Reference) (OtherComp :type Primitive%20Component%20Object%20Reference) (OtherBodyIndex :type Integer) (bFromSweep :type Boolean) (SweepResult :type Hit%20Result%20Structure%20%28by%20ref%29))
  (CallFunction|LaunchTargets OtherActor))

(event OnComponentBeginOverlap(Box4)
  (params (OverlappedComponent :type Primitive%20Component%20Object%20Reference) (OtherActor :type Actor%20Object%20Reference) (OtherComp :type Primitive%20Component%20Object%20Reference) (OtherBodyIndex :type Integer) (bFromSweep :type Boolean) (SweepResult :type Hit%20Result%20Structure%20%28by%20ref%29))
  (CallFunction|LaunchTargets OtherActor))

(event OnComponentBeginOverlap(Sphere1)
  (params (OverlappedComponent :type Primitive%20Component%20Object%20Reference) (OtherActor :type Actor%20Object%20Reference) (OtherComp :type Primitive%20Component%20Object%20Reference) (OtherBodyIndex :type Integer) (bFromSweep :type Boolean) (SweepResult :type Hit%20Result%20Structure%20%28by%20ref%29))
  (bind _asfort_athena_vehicle (Utilities|Casting|CastToFortAthenaVehicle OtherActor)
    (:then
      ;[1 1 1] Launch vehicles
      (Audio|BroadcastAbilityImpactatLocation 0 _asfort_athena_vehicle (+ (* (Math|Vector|GetForwardVector (Math|Rotator|FindLookatRotation (Transformation|GetWorldLocation (Variables|Default|GetSphere1)) (Transformation|GetWorldLocation (Variables|/Script/Engine.Actor|GetRootComponent _asfort_athena_vehicle)))) (Variables|Default|GetLaunchPower)) (Transformation|GetVelocity _asfort_athena_vehicle)) 7.5e+02))
    (:CastFailed
      ;[1 1 1] Launch players
      (bind _asfort_pawn (Utilities|Casting|CastToFortPawn OtherActor))
      (Character|LaunchCharacter _asfort_pawn (* (Math|Vector|GetForwardVector (Math|Rotator|FindLookatRotation (Transformation|GetWorldLocation (Variables|Default|GetSphere1)) (Transformation|GetWorldLocation (Variables|/Script/Engine.Character|GetCapsuleComponent _asfort_pawn)))) (Variables|Default|GetLaunchPower)) false false))))

(event Custom|LaunchTargets
  (params (Target :type Object%20Reference))
  (bind _asfort_athena_vehicle (Utilities|Casting|CastToFortAthenaVehicle Target)
    (:then
      ;[1 1 1] Launch vehicles
      (Audio|BroadcastAbilityImpactatLocation 0 _asfort_athena_vehicle (+ (* (Math|Vector|GetForwardVector (Math|Rotator|FindLookatRotation (Transformation|GetWorldLocation (Variables|Default|GetDefaultSceneRoot)) (Transformation|GetWorldLocation (Variables|/Script/Engine.Actor|GetRootComponent _asfort_athena_vehicle)))) (Variables|Default|GetLaunchPower)) (Transformation|GetVelocity _asfort_athena_vehicle)) 7.5e+02))
    (:CastFailed
      ;[1 1 1] Launch players
      (bind _asfort_pawn (Utilities|Casting|CastToFortPawn Target))
      (Character|LaunchCharacter _asfort_pawn (* (Math|Vector|GetForwardVector (Math|Rotator|FindLookatRotation (Transformation|GetWorldLocation (Variables|Default|GetDefaultSceneRoot)) (Transformation|GetWorldLocation (Variables|/Script/Engine.Character|GetCapsuleComponent _asfort_pawn)))) (Variables|Default|GetLaunchPower)) false false))))

(fn OnRep_Stage1Active
  (Utilities|FlowControl|Sequence
    (:then_1
      (Networking|SwitchHasAuthority
        (:Authority
          (Utilities|Time|SetTimerbyEvent (EventDispatchers|CreateEvent :!event CE_Breathe) 3e+01))))))

(fn OnRep_Stage2Active
  (if (Variables|Default|GetStage2Active)
    (Networking|SwitchHasAuthority
      (:Authority
        (Collision|SetCollisionProfileName (Variables|Default|GetSphere) "OverlapAllDynamic" true :!class /Script/Engine.PrimitiveComponent) ; Overlaps start ticking etc. for the eye updates.
      )
      (:Remote
        (Development|SetHiddeninGame (Variables|Default|GetNotVisibleYetMesh) true false)))))

(fn OnRep_TrackedTarget
  (Networking|SwitchHasAuthority
    (:Remote
      (CallFunction|UpdateTarget))))

(fn UpdateTargetAudioLoop
  (bind _target (Variables|Default|GetTarget))
  (bind _smoothedtarget (Variables|Default|GetSmoothedTarget))
  (bind _returnvalue (- _smoothedtarget _target))
  (bind _returnvalue_2 (Math|Vector|VectorLength _returnvalue))
  (bind _returnvalue_3 (Math|Float|Absolute(Float) _returnvalue_2))
  (bind _targetaudioloop (Variables|Default|GetTargetAudioLoop))
  (Audio|Parameter|SetFloatParameter _targetaudioloop "SquishDelta" (Math|Float|MapRangeClamped _returnvalue_3 0.0 5e+01 0.0 1.0) :!class /Script/Engine.AudioComponent)
  (Audio|Parameter|SetFloatParameter _targetaudioloop "SqueakDelta" (Math|Float|MapRangeClamped _returnvalue_3 8e+01 1e+02 0.0 1.0) :!class /Script/Engine.AudioComponent))

(fn ConstructionScript)
